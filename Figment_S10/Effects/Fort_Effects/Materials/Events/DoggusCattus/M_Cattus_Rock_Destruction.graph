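# Masked, two-sided lit particle material for Cattus rock destruction debris, with castle albedo and normal textures.
# A noise-broken sphere mask around the actor, sized by dynamic HotSpotSize and weighted by normal-map Z, blends in particle color plus burn-edge color.
# Inverted particle alpha drives a noise-ramp dissolve through MF_Dissolve.
# A displacement vector adds a velocity parameter to the transformed vertex normal, scaled by vertex red and dynamic Magnitude.
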Material /Figment_S10/Effects/Fort_Effects/Materials/Events/DoggusCattus/M_Cattus_Rock_Destruction
graph {
 properties {
  bUsedWithMeshParticles=true
  bUsedWithParticleSprites=true
  blendMode=BLEND_Masked
  pixelDepthOffsetMode=PDOM_Legacy
  twoSided=true
  usedShadingModels=Default Lit
 }
 nodes [
  Constant() {r=0.5}
  Constant() {r=0.6}
  n1 = ParticleColor()
  n2 = Multiply(
   A=TextureSample() {samplerType=SAMPLERTYPE_LinearColor, texture=/Figment_S10/Effects/Fort_Effects/BluGlow/HighContrastNoise}.G
   B=SphereMask(
    A=WorldPosition().XYZ
    B=ActorPositionWS()
    Radius=DynamicParameter() {defaultValue=<rgba 1 3500 1 1>, paramNames [Magnitude,HotSpotSize,Param3,Param4]}.HotSpotSize
    Hardness=ScalarParameter() {defaultValue=-0.25, parameterName=hardness}
   )
  )
  n3 = TextureSampleParameter2D() {
   parameterName=Normal Texture
   samplerType=SAMPLERTYPE_Normal
   texture=/Figment_S10/Characters/Enemies/Cattus/Demo/Textures/Cattus_Castle_normal
  }
  n4 = Constant3Vector() {constant=<rgba 0 0 1 0>}
  LinearInterpolate(
   B=Add(A=@n1.RGB, B=VectorParameter() {parameterName=BurnEdgeColor, defaultValue=<rgba 4 1.3463 0.0059 1>}.RGB)
   Alpha=Multiply(A=Power(Base=@n2), B=DotProduct(A=@n3.RGB, B=@n4))
  )
  MaterialFunctionCall!(
   Ramp=TextureSample() {samplerType=SAMPLERTYPE_Masks, texture=/Game/Gadgets/Assets/VinderTech_GliderChute/Glider_Vampire/Textures/T_Glider_Bat_Noises}.G
   Dissolve=OneMinus(None=@n1.A)
   EdgeSharpness
  ) {materialFunction=/Game/Effects/Fort_Effects/Materials/_Functions/MF_Dissolve}
  n5 = TransformPosition(None=VertexNormalWS()) {transformSourceType=TRANSFORMPOSSOURCE_World}
  n6 = VectorParameter() {parameterName=VelocityDirectionMagnitude, defaultValue=<rgba 0 -150 0 1>}
  Multiply(
   A=Multiply(A=VertexColor().R, B=Add(A=@n5, B=@n6.RGB))
   B=DynamicParameter() {defaultValue=<rgba 1 3500 1 1>, paramNames [Magnitude,HotSpotSize,Param3,Param4]}.Magnitude
  )
  ScalarParameter() {defaultValue=650, parameterName=sas}
  TextureSampleParameter2D() {parameterName=Diffuse Texture, texture=/Figment_S10/Characters/Enemies/Cattus/Demo/Textures/Cattus_Castle_albedo}
 ]
}
