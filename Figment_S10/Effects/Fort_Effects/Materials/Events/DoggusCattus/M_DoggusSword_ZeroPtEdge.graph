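# Unlit translucent particle material, named as the Zero Point edge effect for the Doggus sword, that draws an animated energy edge.
# Two tiled, panning noise textures (electric and turbulence) add to a powered inverse-V gradient.
# Particle color tints the result, vertex red blends it toward flat particle color, and DeriveHDRfromLDR boosts it to HDR.
# Opacity combines that pattern with a V falloff shaped by dynamic parameter Edge Opacity, vertex red and particle alpha.
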Material /Figment_S10/Effects/Fort_Effects/Materials/Events/DoggusCattus/M_DoggusSword_ZeroPtEdge
graph {
 properties {
  bUsedWithMeshParticles=true
  bUsedWithParticleSprites=true
  blendMode=BLEND_Translucent
  pixelDepthOffsetMode=PDOM_Legacy
  shadingModel=MSM_Unlit
  usedShadingModels=Unlit
 }
 nodes [
  n1 = ComponentMask(None=TextureCoordinate()) {g=true}
  n2 = ScalarParameter() {defaultValue=10, parameterName=UTile}
  n3 = ScalarParameter() {defaultValue=1, parameterName=VTile}
  n4 = Panner(Coordinate=Multiply(A=TextureCoordinate(), B=AppendVector(A=@n2, B=@n3))) {speedX=-0.4, speedY=0.4}
  n5 = ScalarParameter() {defaultValue=20, parameterName=UTile2}
  n6 = ScalarParameter() {defaultValue=0.5, parameterName=VTile2}
  n7 = Panner(Coordinate=Multiply(A=TextureCoordinate(), B=AppendVector(A=@n5, B=@n6))) {speedX=-1}
  n8 = Add(
   A=Power(Base=OneMinus(None=@n1), Exp=ScalarParameter() {defaultValue=4})
   B=Add(
    A=TextureSample(UVs=@n4) {samplerType=SAMPLERTYPE_LinearColor, texture=/Game/Characters/Player/Male/Medium/Bodies/M_MED_SRDrift/Textures/T_ElectricNoise}.R
    B=TextureSample(UVs=@n7) {texture=/Game/Effects/Fort_Effects/Textures/Plasma/T_Turbulence_Seamless}.R
   )
  )
  n9 = ParticleColor()
  n10 = Multiply(A=@n8, B=@n9.RGB)
  n11 = VertexColor()
  MaterialFunctionCall(LDR Input=Clamp(None=LinearInterpolate(A=@n10, B=@n9.RGB, Alpha=@n11.R)), HDR Tint=@n9.RGB) {materialFunction=&DeriveHDRfromLDR}
  n12 = TextureCoordinate()
  n13 = Power(
   Base=OneMinus(None=ComponentMask(None=@n12) {g=true})
   Exp=DynamicParameter() {paramNames [Edge Opacity, Param2, Param3, Param4]}.Edge Opacity
  )
  Multiply(A=@n9.A, B=LinearInterpolate(A=Multiply(A=@n8, B=@n13), Alpha=VertexColor().R))
 ]
}
