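# Additive unlit, two-sided mesh-particle material for a soft flare, apparently on the DoggusCattus event sword.
# Emissive combines radial sphere-mask glows tinted by particle color with a flare texture sampled per RGB channel at offset UVs for chromatic aberration.
# SplineThicken sets strand width, scaled by a SplineScale dynamic parameter.
# Opacity multiplies particle alpha, depth-faded on higher feature levels, by a sphere mask and near-camera fade.
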Material /Figment_S10/Effects/Fort_Effects/Materials/Events/DoggusCattus/M_Doggus_Sword_Flare
graph {
 properties {
  bUsedWithMeshParticles=true
  blendMode=BLEND_Additive
  pixelDepthOffsetMode=PDOM_Legacy
  shadingModel=MSM_Unlit
  twoSided=true
  usedShadingModels=Unlit
 }
 nodes [
  n1 = Constant() {r=0.5}
  n2 = SphereMask(A=TextureCoordinate(), B=@n1, Radius=@n1) {hardnessPercent=0}
  n3 = Power(Base=@n2) {constExponent=25}
  n4 = ParticleColor()
  n5 = Multiply(A=Power(Base=@n2) {constExponent=4}) {constB=0.15}
  n6 = Add(
   A=TextureCoordinate() {vTiling=0.5}
   B=AppendVector(A=ScalarParameter() {parameterName=U}, B=ScalarParameter() {defaultValue=0.1, parameterName=V})
  )
  n7 = ScalarParameter() {defaultValue=0.01, parameterName=ChromaOffset}
  n8 = Add(A=@n6, B=@n7) {constB=0.03}
  n9 = Multiply(A=@n7) {constB=-1}
  n10 = AppendVector(
   A=AppendVector(
    A=TextureSampleParameter2D(UVs=@n6) {parameterName=FlareMask, texture=/Game/Effects/Fort_Effects/Textures/Flares/T_Flare_Soft}.R
    B=TextureSampleParameter2D(UVs=@n8) {parameterName=FlareMask_1, texture=&T_Flare_Soft}.G
   )
   B=TextureSampleParameter2D(UVs=Add(A=@n6, B=@n9) {constB=-0.03}) {parameterName=FlareMask_2, texture=&T_Flare_Soft}.B
  )
  Add(
   A=Multiply(A=Multiply(A=@n3) {constB=2}, B=LinearInterpolate(B=@n4.RGB) {constA=1, constAlpha=0.7})
   B=Multiply(A=Add(A=@n5, B=@n10), B=@n4.RGB)
  )
  n11 = DynamicParameter() {paramNames [SplineScale,Param2,Param3,Param4]}
  MaterialFunctionCall(
   WidthBase=Multiply(A=ScalarParameter() {defaultValue=15000, parameterName=WidthBase}, B=@n11.SplineScale)
   WidthTip=Multiply(A=ScalarParameter() {defaultValue=5000, parameterName=WidthTip}, B=@n11.SplineScale)
  ) {materialFunction=&SplineThicken}
  n12 = DepthFade(Opacity=@n4.A, FadeDistance=ScalarParameter() {defaultValue=100, parameterName=DepthFade})
  n13 = TextureCoordinate() {uTiling=1.5}
  n14 = ComponentMask(None=VectorParameter() {parameterName=Spheremask_B, defaultValue=<rgba 0.75 0.5 0 0>}.RGB) {g=true, r=true}
  n15 = ScalarParameter() {defaultValue=0.5, parameterName=Radius}
  Multiply(
   A=Multiply(
    A=FeatureLevelSwitch(Default=@n12, ES2=@n4.A, ES3_1=@n4.A, SM4_REMOVED=@n12, SM5=@n12, SM6=@n12)
    B=SphereMask(A=@n13, B=@n14, Radius=@n15) {hardnessPercent=0}
   )
   B=MaterialFunctionCall(Distance=ScalarParameter() {defaultValue=100, parameterName=NearFadeDistance}) {materialFunction=&Near_Camera_Fade}.Result
  )
 ]
}
