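# Translucent unlit, two-sided material for the Zero Point energy sword splines in the DoggusCattus event, usable on mesh particles.
# SplineThicken widens strands with a half-width tip, plus a local-Z offset weighted by inverted V.
# Color sums panning magnetic waves, color-fringed caustics from offset samples and a tint, with vertex red offsetting UVs.
# Opacity multiplies that color by UV edge masks, view-angle and depth fades, and a UV1-plus-SwordAlpha term.
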Material /Figment_S10/Effects/Fort_Effects/Materials/Events/DoggusCattus/M_Zeropt_EnergySwordSplines
graph {
 properties {
  bUsedWithMeshParticles=true
  blendMode=BLEND_Translucent
  pixelDepthOffsetMode=PDOM_Legacy
  shadingModel=MSM_Unlit
  twoSided=true
  usedShadingModels=Unlit
 }
 nodes [
  n1 = ScalarParameter() {defaultValue=1000, parameterName=WidthBase}
  n2 = Transform(None=Constant3Vector() {constant=<rgba 0 0 1 0>}) {transformSourceType=TRANSFORMSOURCE_Local}
  n3 = TextureCoordinate()
  n4 = Power(Base=OneMinus(None=ComponentMask(None=@n3) {g=true}), Exp=ScalarParameter() {defaultValue=2, parameterName=WPO_Exp})
  Add(
   A=MaterialFunctionCall(WidthBase=@n1, WidthTip=Multiply(A=@n1) {constB=0.5}) {materialFunction=&SplineThicken}.WorldPositionOffset
   B=Multiply(A=Multiply(A=@n2, B=@n4), B=ScalarParameter() {defaultValue=20, parameterName=WPO_Amount})
  )
  n5 = VertexColor()
  n6 = VectorParameter() {parameterName=Tile1, defaultValue=<rgba 0.1 1 0 0>}
  n7 = AppendVector(A=@n6.R, B=@n6.R)
  n8 = Multiply(B=@n7) {constA=-0.2, constB=0.5}
  n9 = Multiply(A=Add(A=TextureCoordinate() {coordinateIndex=1}, B=Add(A=@n5.R, B=@n8)), B=@n7)
  n10 = VectorParameter() {parameterName=PanDirection1, defaultValue=<rgba 0.02 0.2 0 0>}
  n11 = AppendVector(A=@n10.R, B=@n10.R)
  n12 = Add(A=Multiply(A=Time(), B=ScalarParameter() {defaultValue=0.5, parameterName=PanSpeed}), B=ScalarParameter() {parameterName=Random})
  n13 = TextureSample(UVs=Panner(Coordinate=Multiply(A=@n9) {constB=1.5}, Time=@n12, Speed=@n11) {speedX=-0.2, speedY=-0.02}) {texture=/Game/Athena/Prototype/Blueprints/Sneezy/Textures/T_Sneezy_MagneticWaves}
  n14 = TextureSample(UVs=Panner(Coordinate=@n9, Time=@n12, Speed=@n11) {speedX=-0.2, speedY=0.02}) {texture=&T_Sneezy_MagneticWaves}
  n15 = TextureCoordinate()
  n16 = VectorParameter() {parameterName=Tile2, defaultValue=<rgba 0.3 0.2 0 0>}
  n17 = AppendVector(A=@n16.R, B=@n16.R)
  n18 = VectorParameter() {parameterName=PanDirection2, defaultValue=<rgba 0.04 0.1 0 0>}
  n19 = Panner(Coordinate=Add(A=Multiply(A=@n15, B=@n17), B=VertexColor().R), Time=@n12, Speed=AppendVector(A=@n18.R, B=@n18.R)) {speedX=0.1, speedY=0.04}
  n20 = Multiply(
   A=Constant3Vector() {constant=<rgba 1 0 0.4 0>}
   B=TextureSample(UVs=Add(A=@n19) {constB=0.01}) {texture=/Figment_S10/Effects/Fort_Effects/Textures/Water/T_Caustics_Softer}.G
  )
  n21 = Multiply(
   A=Constant3Vector() {constant=<rgba 0 1 0.6 0>}
   B=TextureSample(UVs=@n19) {texture=&T_Caustics_Softer}.G
  )
  n22 = Add(
   A=Add(A=Add(A=@n13.RGB, B=@n14.RGB), B=LinearInterpolate(A=@n20, B=@n21))
   B=VectorParameter() {parameterName=Color, defaultValue=<rgba 0.0771 0.2812 0.4 0>}.RGB
  )
  n23 = Multiply(A=VectorParameter() {parameterName=HDR_Color, defaultValue=<rgba 1 1 1 0>}.RGB, B=@n22)
  MaterialFunctionCall(LDR Input=Clamp(None=@n23), HDR Tint=@n23) {materialFunction=&DeriveHDRfromLDR}
  n24 = Power(Base=ComponentMask(None=TextureCoordinate()) {g=true}, Exp=ScalarParameter() {defaultValue=2, parameterName=TopExp})
  n25 = TextureCoordinate()
  n26 = Power(Base=Sine(None=ComponentMask(None=@n25) {r=true}) {period=2}, Exp=ScalarParameter() {defaultValue=2, parameterName=SideExp})
  n27 = Saturate(None=Multiply(A=Multiply(A=@n24, B=@n26)) {constB=10})
  n28 = MaterialFunctionCall("FadeSensitivity (optional)"=ScalarParameter() {defaultValue=1, parameterName=AngleFade}) {materialFunction=&FadeBasedOnViewAngle}
  n29 = Multiply(A=Saturate(None=Multiply(A=@n28.Result) {constB=2}), B=DepthFade(FadeDistance=ScalarParameter() {defaultValue=125, parameterName=DepthFade}))
  n30 = ComponentMask(None=TextureCoordinate() {coordinateIndex=1}) {g=true}
  n31 = ScalarParameter() {defaultValue=1, parameterName=SwordAlpha}
  Multiply(A=Multiply(A=@n22, B=Multiply(A=@n27, B=@n29)), B=Clamp(None=Add(A=@n30, B=@n31)))
 ]
}
